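; An editor utility that logs how many instances of chosen actor classes exist in the level.
; Its Run Logging custom event loops over a configurable array of classes and gets all actors of each class.
; For each it prints the class display name and count to screen and log in light blue.
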(event Custom|Run%20Logging
  (for _array_element (Variables|Default|GetActorClassesToCount)
    (bind _outactors (Actor|GetAllActorsOfClass _array_element))
    (Development|PrintString (Utilities|String|Append (Utilities|GetClassDisplayName _array_element) " // " (Utilities|String|ToString(Integer) (Utilities|Array|Length _outactors))) true true "(R=0.000000,G=0.660000,B=1.000000,A=1.000000)" 1e+01)))

(fn ConstructionScript)
